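# Sound cue for the Slinky vehicle mod's suspension bounce, branching on IsLocal between first-person and third-person attenuation.
# Each branch mixes randomized suspension lower and raise one-shots with volume modulation, plus pitch variation on the raise.
# A delayed body rattle layer plays through quality-level nodes whose last slot is empty, and is modulated only in first person.
# Concurrency override caps simultaneous instances, and virtualization is disabled.
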SoundCue /VehicleMods/Mods/Slinky/Audio/SC_Slinky_Susp_Bounce
root=Branch_0
nodes [
 {id=Attenuation_0, children [Mixer_0], properties {attenuationSettings=&/ATT_Slinky_1P}}
 {id=Attenuation_1, children [Mixer_1], properties {attenuationSettings=&/ATT_Slinky_3P}}
 {id=Branch_0, children [Attenuation_0, Mixer_2, Mixer_2], properties {boolParameterName=IsLocal}}
 {id=Delay_0, children [Modulator_2]}
 {id=Delay_1, children [Modulator_0]}
 {id=Delay_3, children [Random_5], properties {delayMax=0.75, delayMin=0.75}}
 {id=Delay_4, children [Modulator_4], properties {delayMax=0.75, delayMin=0.75}}
 {id=Mixer_0, children [Modulator_1, Delay_0, QualityLevel_2], properties {inputVolume [1,1,0.75]}}
 {id=Mixer_1, children [Modulator_3, Delay_1, QualityLevel_1], properties {inputVolume [1,1,1]}}
 {id=Mixer_2, children [Attenuation_1], properties {inputVolume [1]}}
 {id=Modulator_0, children [Random_1], properties {pitchMax=1.1, pitchMin=0.925, volumeMax=1, volumeMin=0.6}}
 {id=Modulator_1, children [Random_2], properties {volumeMax=1, volumeMin=0.4}}
 {id=Modulator_2, children [Random_0], properties {pitchMax=1.1, pitchMin=0.925, volumeMax=1, volumeMin=0.6}}
 {id=Modulator_3, children [Random_3], properties {volumeMax=1, volumeMin=0.4}}
 {id=Modulator_4, children [Random_6], properties {pitchMax=1.2, volumeMax=1, volumeMin=0.5}}
 {id=QualityLevel_1, children [Delay_3,Delay_3,Delay_3,null]}
 {id=QualityLevel_2, children [Delay_4,Delay_4,Delay_4,null]}
 {
  id=Random_0
  children [WavePlayer_3,WavePlayer_4,WavePlayer_5,WavePlayer_6,WavePlayer_7,WavePlayer_8]
  properties {bShouldExcludeFromBranchCulling=true, weights [1,1,1,1,1,1]}
 }
 {
  id=Random_1
  children [WavePlayer_0, WavePlayer_1, WavePlayer_2, WavePlayer_9, WavePlayer_10, WavePlayer_11]
  properties {bShouldExcludeFromBranchCulling=true, weights [1,1,1,1,1,1]}
 }
 {
  id=Random_2
  children [WavePlayer_12, WavePlayer_13, WavePlayer_14, WavePlayer_15]
  properties {bShouldExcludeFromBranchCulling=true, weights [1,1,1,1]}
 }
 {
  id=Random_3
  children [WavePlayer_16, WavePlayer_17, WavePlayer_18, WavePlayer_19]
  properties {bShouldExcludeFromBranchCulling=true, weights [1,1,1,1]}
 }
 {
  id=Random_5
  children [
   WavePlayer_31
   WavePlayer_32
   WavePlayer_33
   WavePlayer_34
   WavePlayer_35
   WavePlayer_36
   WavePlayer_37
   WavePlayer_38
   WavePlayer_39
   WavePlayer_40
   WavePlayer_41
  ]
  properties {bShouldExcludeFromBranchCulling=true, weights [1,1,1,1,1,1,1,1,1,1,1]}
 }
 {
  id=Random_6
  children [
   WavePlayer_42
   WavePlayer_43
   WavePlayer_44
   WavePlayer_45
   WavePlayer_46
   WavePlayer_47
   WavePlayer_48
   WavePlayer_49
   WavePlayer_50
   WavePlayer_51
   WavePlayer_52
  ]
  properties {bShouldExcludeFromBranchCulling=true, weights [1,1,1,1,1,1,1,1,1,1,1]}
 }
 {id=WavePlayer_0, properties {soundWaveAssetPtr=&/Slinky_Susp_Adjust_Raise_Large_01}}
 {id=WavePlayer_1, properties {soundWaveAssetPtr=&/Slinky_Susp_Adjust_Raise_Large_02}}
 {id=WavePlayer_10, properties {soundWaveAssetPtr=&/Slinky_Susp_Adjust_Raise_Large_05}}
 {id=WavePlayer_11, properties {soundWaveAssetPtr=&/Slinky_Susp_Adjust_Raise_Large_06}}
 {id=WavePlayer_12, properties {soundWaveAssetPtr=&/Slinky_Susp_Adjust_Lower_Large_01}}
 {id=WavePlayer_13, properties {soundWaveAssetPtr=&/Slinky_Susp_Adjust_Lower_Large_02}}
 {id=WavePlayer_14, properties {soundWaveAssetPtr=&/Slinky_Susp_Adjust_Lower_Large_03}}
 {id=WavePlayer_15, properties {soundWaveAssetPtr=&/Slinky_Susp_Adjust_Lower_Large_04}}
 {id=WavePlayer_16, properties {soundWaveAssetPtr=&Slinky_Susp_Adjust_Lower_Large_01}}
 {id=WavePlayer_17, properties {soundWaveAssetPtr=&Slinky_Susp_Adjust_Lower_Large_02}}
 {id=WavePlayer_18, properties {soundWaveAssetPtr=&Slinky_Susp_Adjust_Lower_Large_03}}
 {id=WavePlayer_19, properties {soundWaveAssetPtr=&Slinky_Susp_Adjust_Lower_Large_04}}
 {id=WavePlayer_2, properties {soundWaveAssetPtr=&/Slinky_Susp_Adjust_Raise_Large_03}}
 {id=WavePlayer_3, properties {soundWaveAssetPtr=&Slinky_Susp_Adjust_Raise_Large_01}}
 {id=WavePlayer_31, properties {soundWaveAssetPtr=&/Slinky_Susp_Adjust_Body_Rattle_Large_01}}
 {id=WavePlayer_32, properties {soundWaveAssetPtr=&/Slinky_Susp_Adjust_Body_Rattle_Large_02}}
 {id=WavePlayer_33, properties {soundWaveAssetPtr=&/Slinky_Susp_Adjust_Body_Rattle_Large_03}}
 {id=WavePlayer_34, properties {soundWaveAssetPtr=&/Slinky_Susp_Adjust_Body_Rattle_Large_04}}
 {id=WavePlayer_35, properties {soundWaveAssetPtr=&/Slinky_Susp_Adjust_Body_Rattle_Large_05}}
 {id=WavePlayer_36, properties {soundWaveAssetPtr=&/Slinky_Susp_Adjust_Body_Rattle_Large_06}}
 {id=WavePlayer_37, properties {soundWaveAssetPtr=&/Slinky_Susp_Adjust_Body_Rattle_Large_07}}
 {id=WavePlayer_38, properties {soundWaveAssetPtr=&/Slinky_Susp_Adjust_Body_Rattle_Large_08}}
 {id=WavePlayer_39, properties {soundWaveAssetPtr=&/Slinky_Susp_Adjust_Body_Rattle_Large_09}}
 {id=WavePlayer_4, properties {soundWaveAssetPtr=&Slinky_Susp_Adjust_Raise_Large_02}}
 {id=WavePlayer_40, properties {soundWaveAssetPtr=&/Slinky_Susp_Adjust_Body_Rattle_Large_10}}
 {id=WavePlayer_41, properties {soundWaveAssetPtr=&/Slinky_Susp_Adjust_Body_Rattle_Large_11}}
 {id=WavePlayer_42, properties {soundWaveAssetPtr=&Slinky_Susp_Adjust_Body_Rattle_Large_01}}
 {id=WavePlayer_43, properties {soundWaveAssetPtr=&Slinky_Susp_Adjust_Body_Rattle_Large_02}}
 {id=WavePlayer_44, properties {soundWaveAssetPtr=&Slinky_Susp_Adjust_Body_Rattle_Large_03}}
 {id=WavePlayer_45, properties {soundWaveAssetPtr=&Slinky_Susp_Adjust_Body_Rattle_Large_04}}
 {id=WavePlayer_46, properties {soundWaveAssetPtr=&Slinky_Susp_Adjust_Body_Rattle_Large_05}}
 {id=WavePlayer_47, properties {soundWaveAssetPtr=&Slinky_Susp_Adjust_Body_Rattle_Large_06}}
 {id=WavePlayer_48, properties {soundWaveAssetPtr=&Slinky_Susp_Adjust_Body_Rattle_Large_07}}
 {id=WavePlayer_49, properties {soundWaveAssetPtr=&Slinky_Susp_Adjust_Body_Rattle_Large_08}}
 {id=WavePlayer_5, properties {soundWaveAssetPtr=&Slinky_Susp_Adjust_Raise_Large_03}}
 {id=WavePlayer_50, properties {soundWaveAssetPtr=&Slinky_Susp_Adjust_Body_Rattle_Large_09}}
 {id=WavePlayer_51, properties {soundWaveAssetPtr=&Slinky_Susp_Adjust_Body_Rattle_Large_10}}
 {id=WavePlayer_52, properties {soundWaveAssetPtr=&Slinky_Susp_Adjust_Body_Rattle_Large_11}}
 {id=WavePlayer_6, properties {soundWaveAssetPtr=&/Slinky_Susp_Adjust_Raise_Large_04}}
 {id=WavePlayer_7, properties {soundWaveAssetPtr=&Slinky_Susp_Adjust_Raise_Large_05}}
 {id=WavePlayer_8, properties {soundWaveAssetPtr=&Slinky_Susp_Adjust_Raise_Large_06}}
 {id=WavePlayer_9, properties {soundWaveAssetPtr=&Slinky_Susp_Adjust_Raise_Large_04}}
]
properties {
 bOverrideConcurrency=true
 maxDistance=12000
 priority=0.8
 soundClassObject=&Fort_Vehicles_Exterior
 virtualizationMode=Disabled
 volumeMultiplier=0.6
 concurrencyOverrides {maxCount=3}
}
