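# Sound cue for the Slinky vehicle mod's suspension jump, layering a random ascend sound with a random thump.
# An IsLocal branch selects first-person attenuation for the local case and third-person jump attenuation when false or unset.
# Both paths mix the same sample sets, boosting the thump layer and slightly randomizing its pitch and volume.
# It plays in the exterior vehicle sound class with capped concurrency and virtualization disabled.
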SoundCue /VehicleMods/Mods/Slinky/Audio/SC_Slinky_Susp_Jump
root=Branch_0
nodes [
 {id=Attenuation_0, children [Mixer_1], properties {attenuationSettings=&/ATT_Slinky_1P}}
 {id=Attenuation_1, children [Mixer_3], properties {attenuationSettings=&/ATT_Slinky_Jump_3P}}
 {id=Branch_0, children [Attenuation_0, Mixer_2, Mixer_2], properties {boolParameterName=IsLocal}}
 {id=Mixer_1, children [Random_2,Modulator_1], properties {inputVolume [1,1.5]}}
 {id=Mixer_2, children [Attenuation_1], properties {inputVolume [1]}}
 {id=Mixer_3, children [Random_4,Modulator_3], properties {inputVolume [1,1.5]}}
 {id=Modulator_1, children [Random_3], properties {pitchMax=1.2, pitchMin=1, volumeMax=1, volumeMin=0.9}}
 {id=Modulator_3, children [Random_5], properties {pitchMax=1.2, pitchMin=1, volumeMax=1, volumeMin=0.9}}
 {
  id=Random_2
  children [WavePlayer_12, WavePlayer_13, WavePlayer_14, WavePlayer_15, WavePlayer_16, WavePlayer_17]
  properties {bShouldExcludeFromBranchCulling=true, weights [1,1,1,1,1,1]}
 }
 {id=Random_3, children [WavePlayer_18, WavePlayer_19, WavePlayer_20], properties {bShouldExcludeFromBranchCulling=true, weights [1,1,1]}}
 {
  id=Random_4
  children [WavePlayer_21, WavePlayer_22, WavePlayer_23, WavePlayer_24, WavePlayer_25, WavePlayer_26]
  properties {bShouldExcludeFromBranchCulling=true, weights [1,1,1,1,1,1]}
 }
 {id=Random_5, children [WavePlayer_27, WavePlayer_28, WavePlayer_29], properties {bShouldExcludeFromBranchCulling=true, weights [1,1,1]}}
 {id=WavePlayer_12, properties {soundWaveAssetPtr=&/Slinky_Susp_Jump_Ascend_06}}
 {id=WavePlayer_13, properties {soundWaveAssetPtr=&/Slinky_Susp_Jump_Ascend_05}}
 {id=WavePlayer_14, properties {soundWaveAssetPtr=&/Slinky_Susp_Jump_Ascend_04}}
 {id=WavePlayer_15, properties {soundWaveAssetPtr=&/Slinky_Susp_Jump_Ascend_03}}
 {id=WavePlayer_16, properties {soundWaveAssetPtr=&/Slinky_Susp_Jump_Ascend_02}}
 {id=WavePlayer_17, properties {soundWaveAssetPtr=&/Slinky_Susp_Jump_Ascend_01}}
 {id=WavePlayer_18, properties {soundWaveAssetPtr=&/Slinky_Susp_Jump_Ascend_Thump_01}}
 {id=WavePlayer_19, properties {soundWaveAssetPtr=&/Slinky_Susp_Jump_Ascend_Thump_02}}
 {id=WavePlayer_20, properties {soundWaveAssetPtr=&/Slinky_Susp_Jump_Ascend_Thump_03}}
 {id=WavePlayer_21, properties {soundWaveAssetPtr=&Slinky_Susp_Jump_Ascend_06}}
 {id=WavePlayer_22, properties {soundWaveAssetPtr=&Slinky_Susp_Jump_Ascend_05}}
 {id=WavePlayer_23, properties {soundWaveAssetPtr=&Slinky_Susp_Jump_Ascend_04}}
 {id=WavePlayer_24, properties {soundWaveAssetPtr=&Slinky_Susp_Jump_Ascend_03}}
 {id=WavePlayer_25, properties {soundWaveAssetPtr=&Slinky_Susp_Jump_Ascend_02}}
 {id=WavePlayer_26, properties {soundWaveAssetPtr=&Slinky_Susp_Jump_Ascend_01}}
 {id=WavePlayer_27, properties {soundWaveAssetPtr=&Slinky_Susp_Jump_Ascend_Thump_01}}
 {id=WavePlayer_28, properties {soundWaveAssetPtr=&Slinky_Susp_Jump_Ascend_Thump_02}}
 {id=WavePlayer_29, properties {soundWaveAssetPtr=&Slinky_Susp_Jump_Ascend_Thump_03}}
]
properties {
 bOverrideConcurrency=true
 maxDistance=12000
 priority=0.8
 soundClassObject=&Fort_Vehicles_Exterior
 virtualizationMode=Disabled
 volumeMultiplier=0.5
 concurrencyOverrides {maxCount=3}
}
