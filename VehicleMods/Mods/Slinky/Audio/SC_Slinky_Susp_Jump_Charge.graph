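# Sound cue for the Slinky vehicle mod's suspension jump charge, routed to the exterior vehicle sound class with capped concurrency.
# An IsLocal branch selects first-person attenuation for the local player and third-person attenuation when false or unset.
# Both paths mix the same layers: one of six random charge waves, a modulated swirl and a quieter modulated air layer.
# The random nodes are excluded from branch culling.
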SoundCue /VehicleMods/Mods/Slinky/Audio/SC_Slinky_Susp_Jump_Charge
root=Branch_0
nodes [
 {id=Attenuation_0, children [Mixer_0], properties {attenuationSettings=&/ATT_Slinky_1P}}
 {id=Attenuation_1, children [Mixer_2], properties {attenuationSettings=&/ATT_Slinky_3P}}
 {id=Branch_0, children [Attenuation_0, Mixer_1, Mixer_1], properties {boolParameterName=IsLocal}}
 {id=Mixer_0, children [Random_4,Modulator_0,Modulator_3], properties {inputVolume [1,1,0.5]}}
 {id=Mixer_1, children [Attenuation_1], properties {inputVolume [1]}}
 {id=Mixer_2, children [Random_0,Modulator_1,Modulator_2], properties {inputVolume [1,1,0.5]}}
 {id=Modulator_0, children [WavePlayer_0]}
 {id=Modulator_1, children [WavePlayer_7]}
 {id=Modulator_2, children [WavePlayer_8]}
 {id=Modulator_3, children [WavePlayer_9]}
 {
  id=Random_0
  children [WavePlayer_1,WavePlayer_2,WavePlayer_3,WavePlayer_4,WavePlayer_5,WavePlayer_6]
  properties {bShouldExcludeFromBranchCulling=true, weights [1,1,1,1,1,1]}
 }
 {
  id=Random_4
  children [WavePlayer_24, WavePlayer_25, WavePlayer_26, WavePlayer_27, WavePlayer_28, WavePlayer_29]
  properties {bShouldExcludeFromBranchCulling=true, weights [1,1,1,1,1,1]}
 }
 {id=WavePlayer_0, properties {soundWaveAssetPtr=&/Slinky_Susp_Jump_Charge_Swirl_01}}
 {id=WavePlayer_1, properties {soundWaveAssetPtr=&/Slinky_Susp_Jump_Charge_01}}
 {id=WavePlayer_2, properties {soundWaveAssetPtr=&/Slinky_Susp_Jump_Charge_02}}
 {id=WavePlayer_24, properties {soundWaveAssetPtr=&Slinky_Susp_Jump_Charge_01}}
 {id=WavePlayer_25, properties {soundWaveAssetPtr=&Slinky_Susp_Jump_Charge_02}}
 {id=WavePlayer_26, properties {soundWaveAssetPtr=&/Slinky_Susp_Jump_Charge_03}}
 {id=WavePlayer_27, properties {soundWaveAssetPtr=&/Slinky_Susp_Jump_Charge_04}}
 {id=WavePlayer_28, properties {soundWaveAssetPtr=&/Slinky_Susp_Jump_Charge_05}}
 {id=WavePlayer_29, properties {soundWaveAssetPtr=&/Slinky_Susp_Jump_Charge_06}}
 {id=WavePlayer_3, properties {soundWaveAssetPtr=&Slinky_Susp_Jump_Charge_03}}
 {id=WavePlayer_4, properties {soundWaveAssetPtr=&Slinky_Susp_Jump_Charge_04}}
 {id=WavePlayer_5, properties {soundWaveAssetPtr=&Slinky_Susp_Jump_Charge_05}}
 {id=WavePlayer_6, properties {soundWaveAssetPtr=&Slinky_Susp_Jump_Charge_06}}
 {id=WavePlayer_7, properties {soundWaveAssetPtr=&Slinky_Susp_Jump_Charge_Swirl_01}}
 {id=WavePlayer_8, properties {soundWaveAssetPtr=&/Slinky_Susp_Jump_Charge_Air_01}}
 {id=WavePlayer_9, properties {soundWaveAssetPtr=&Slinky_Susp_Jump_Charge_Air_01}}
]
properties {
 bOverrideConcurrency=true
 maxDistance=12000
 priority=0.8
 soundClassObject=&Fort_Vehicles_Exterior
 virtualizationMode=Disabled
 volumeMultiplier=0.5
 concurrencyOverrides {maxCount=3}
}
