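# Sound cue for the Slinky vehicle mod's suspension jump prime sound, playing one wave through a modulator in the exterior vehicle sound class.
# An IsLocal branch sends local playback through first-person attenuation.
# The false and unset cases share a mixer into third-person attenuation, which has a first-person attenuation node nested beneath it.
# Two concurrency assets limit instances, and virtualization is disabled.
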SoundCue /VehicleMods/Mods/Slinky/Audio/SC_Slinky_Susp_Jump_Prime
root=Branch_0
nodes [
 {id=Attenuation_0, children [Modulator_0], properties {attenuationSettings=&/ATT_Slinky_1P}}
 {id=Attenuation_1, children [Attenuation_2], properties {attenuationSettings=&/ATT_Slinky_3P}}
 {id=Attenuation_2, children [Modulator_3], properties {attenuationSettings=&ATT_Slinky_1P}}
 {id=Branch_0, children [Attenuation_0, Mixer_1, Mixer_1], properties {boolParameterName=IsLocal}}
 {id=Mixer_1, children [Attenuation_1], properties {inputVolume [1]}}
 {id=Modulator_0, children [WavePlayer_0]}
 {id=Modulator_3, children [WavePlayer_3]}
 {id=WavePlayer_0, properties {soundWaveAssetPtr=&/Slinky_Susp_Jump_Prime_01}}
 {id=WavePlayer_3, properties {soundWaveAssetPtr=&Slinky_Susp_Jump_Prime_01}}
]
properties {
 maxDistance=12000
 priority=0.8
 soundClassObject=&Fort_Vehicles_Exterior
 virtualizationMode=Disabled
 volumeMultiplier=0.5
 concurrencyOverrides {maxCount=2}
 concurrencySet [&/CONC_Slinky_Local, &/CONC_Slinky_Kill_Oldest]
}
