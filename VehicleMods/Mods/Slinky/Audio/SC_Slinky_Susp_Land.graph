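# Sound cue for the landing sound of the Slinky vehicle mod's suspension, routed to the exterior vehicle sound class.
# An IsLocal branch selects first-person attenuation when true, or third-person attenuation when false or unset.
# Each path mixes a random jump-ascend wave, excluded from branch culling, with a random, modulated jump-land wave.
# A concurrency override caps simultaneous instances, and virtualization is disabled.
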SoundCue /VehicleMods/Mods/Slinky/Audio/SC_Slinky_Susp_Land
root=Branch_0
nodes [
 {id=Attenuation_0, children [Mixer_1], properties {attenuationSettings=&/ATT_Slinky_1P}}
 {id=Attenuation_1, children [Mixer_0], properties {attenuationSettings=&/ATT_Slinky_3P}}
 {id=Branch_0, children [Attenuation_0, Mixer_2, Mixer_2], properties {boolParameterName=IsLocal}}
 {id=Mixer_0, children [Random_1,Modulator_2], properties {inputVolume [1,1]}}
 {id=Mixer_1, children [Random_2,Modulator_0], properties {inputVolume [1,1]}}
 {id=Mixer_2, children [Attenuation_1], properties {inputVolume [1]}}
 {id=Modulator_0, children [Random_0]}
 {id=Modulator_2, children [Random_6]}
 {id=Random_0, children [WavePlayer_0,WavePlayer_1,WavePlayer_2,WavePlayer_3], properties {weights [1,1,1,1]}}
 {
  id=Random_1
  children [WavePlayer_4,WavePlayer_5,WavePlayer_6,WavePlayer_7,WavePlayer_8,WavePlayer_9]
  properties {bShouldExcludeFromBranchCulling=true, weights [1,1,1,1,1,1]}
 }
 {
  id=Random_2
  children [WavePlayer_12, WavePlayer_13, WavePlayer_14, WavePlayer_15, WavePlayer_16, WavePlayer_17]
  properties {bShouldExcludeFromBranchCulling=true, weights [1,1,1,1,1,1]}
 }
 {id=Random_6, children [WavePlayer_10, WavePlayer_11, WavePlayer_30, WavePlayer_31], properties {weights [1,1,1,1]}}
 {id=WavePlayer_0, properties {soundWaveAssetPtr=&/Slinky_Susp_Jump_Land_01}}
 {id=WavePlayer_1, properties {soundWaveAssetPtr=&/Slinky_Susp_Jump_Land_02}}
 {id=WavePlayer_10, properties {soundWaveAssetPtr=&Slinky_Susp_Jump_Land_01}}
 {id=WavePlayer_11, properties {soundWaveAssetPtr=&Slinky_Susp_Jump_Land_02}}
 {id=WavePlayer_12, properties {soundWaveAssetPtr=&/Slinky_Susp_Jump_Ascend_06}}
 {id=WavePlayer_13, properties {soundWaveAssetPtr=&/Slinky_Susp_Jump_Ascend_05}}
 {id=WavePlayer_14, properties {soundWaveAssetPtr=&/Slinky_Susp_Jump_Ascend_04}}
 {id=WavePlayer_15, properties {soundWaveAssetPtr=&/Slinky_Susp_Jump_Ascend_03}}
 {id=WavePlayer_16, properties {soundWaveAssetPtr=&/Slinky_Susp_Jump_Ascend_02}}
 {id=WavePlayer_17, properties {soundWaveAssetPtr=&/Slinky_Susp_Jump_Ascend_01}}
 {id=WavePlayer_2, properties {soundWaveAssetPtr=&/Slinky_Susp_Jump_Land_03}}
 {id=WavePlayer_3, properties {soundWaveAssetPtr=&/Slinky_Susp_Jump_Land_04}}
 {id=WavePlayer_30, properties {soundWaveAssetPtr=&Slinky_Susp_Jump_Land_03}}
 {id=WavePlayer_31, properties {soundWaveAssetPtr=&Slinky_Susp_Jump_Land_04}}
 {id=WavePlayer_4, properties {soundWaveAssetPtr=&Slinky_Susp_Jump_Ascend_06}}
 {id=WavePlayer_5, properties {soundWaveAssetPtr=&Slinky_Susp_Jump_Ascend_05}}
 {id=WavePlayer_6, properties {soundWaveAssetPtr=&Slinky_Susp_Jump_Ascend_04}}
 {id=WavePlayer_7, properties {soundWaveAssetPtr=&Slinky_Susp_Jump_Ascend_03}}
 {id=WavePlayer_8, properties {soundWaveAssetPtr=&Slinky_Susp_Jump_Ascend_02}}
 {id=WavePlayer_9, properties {soundWaveAssetPtr=&Slinky_Susp_Jump_Ascend_01}}
]
properties {
 bOverrideConcurrency=true
 maxDistance=12000
 priority=0.8
 soundClassObject=&Fort_Vehicles_Exterior
 virtualizationMode=Disabled
 volumeMultiplier=0.5
 concurrencyOverrides {maxCount=3}
}
