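# Sound cue for a large lowering adjustment of the Slinky vehicle mod's suspension.
# An IsLocal branch selects first-person attenuation when true, third-person when false or unset.
# Each path mixes a random lowering sound with a random body rattle, both pitch-modulated, with rattle volume also varying.
# Concurrency is limited per owner, stopping the oldest instance, and virtualization is disabled.
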SoundCue /VehicleMods/Mods/Slinky/Audio/SC_Slinky_Susp_Lower_Large
root=Branch_0
nodes [
 {id=Attenuation_0, children [Mixer_0], properties {attenuationSettings=&/ATT_Slinky_1P}}
 {id=Attenuation_1, children [Mixer_2], properties {attenuationSettings=&/ATT_Slinky_3P}}
 {id=Branch_0, children [Attenuation_0, Mixer_1, Mixer_1], properties {boolParameterName=IsLocal}}
 {id=Mixer_0, children [Modulator_2,Modulator_0], properties {inputVolume [0.5,0.6]}}
 {id=Mixer_1, children [Attenuation_1], properties {inputVolume [1]}}
 {id=Mixer_2, children [Modulator_1,Modulator_3], properties {inputVolume [0.5,0.6]}}
 {id=Modulator_0, children [Random_1], properties {pitchMin=0.9, volumeMax=1, volumeMin=0.3}}
 {id=Modulator_1, children [Random_2], properties {pitchMax=1.1, volumeMax=1, volumeMin=1}}
 {id=Modulator_2, children [Random_0], properties {pitchMax=1.1, volumeMax=1, volumeMin=1}}
 {id=Modulator_3, children [Random_3], properties {pitchMin=0.9, volumeMax=1, volumeMin=0.3}}
 {
  id=Random_0
  children [WavePlayer_3,WavePlayer_4,WavePlayer_5,WavePlayer_6]
  properties {bShouldExcludeFromBranchCulling=true, weights [1,1,1,1]}
 }
 {
  id=Random_1
  children [
   WavePlayer_0
   WavePlayer_1
   WavePlayer_2
   WavePlayer_7
   WavePlayer_8
   WavePlayer_9
   WavePlayer_10
   WavePlayer_11
   WavePlayer_12
   WavePlayer_13
   WavePlayer_14
  ]
  properties {bShouldExcludeFromBranchCulling=true, weights [1,1,1,1,1,1,1,1,1,1,1]}
 }
 {
  id=Random_2
  children [WavePlayer_15, WavePlayer_16, WavePlayer_17, WavePlayer_18]
  properties {bShouldExcludeFromBranchCulling=true, weights [1,1,1,1]}
 }
 {
  id=Random_3
  children [
   WavePlayer_19
   WavePlayer_20
   WavePlayer_21
   WavePlayer_22
   WavePlayer_23
   WavePlayer_24
   WavePlayer_25
   WavePlayer_26
   WavePlayer_27
   WavePlayer_28
   WavePlayer_29
  ]
  properties {bShouldExcludeFromBranchCulling=true, weights [1,1,1,1,1,1,1,1,1,1,1]}
 }
 {id=WavePlayer_0, properties {soundWaveAssetPtr=&/Slinky_Susp_Adjust_Body_Rattle_Large_01}}
 {id=WavePlayer_1, properties {soundWaveAssetPtr=&/Slinky_Susp_Adjust_Body_Rattle_Large_02}}
 {id=WavePlayer_10, properties {soundWaveAssetPtr=&/Slinky_Susp_Adjust_Body_Rattle_Large_07}}
 {id=WavePlayer_11, properties {soundWaveAssetPtr=&/Slinky_Susp_Adjust_Body_Rattle_Large_08}}
 {id=WavePlayer_12, properties {soundWaveAssetPtr=&/Slinky_Susp_Adjust_Body_Rattle_Large_09}}
 {id=WavePlayer_13, properties {soundWaveAssetPtr=&/Slinky_Susp_Adjust_Body_Rattle_Large_10}}
 {id=WavePlayer_14, properties {soundWaveAssetPtr=&/Slinky_Susp_Adjust_Body_Rattle_Large_11}}
 {id=WavePlayer_15, properties {soundWaveAssetPtr=&/Slinky_Susp_Adjust_Lower_Large_01}}
 {id=WavePlayer_16, properties {soundWaveAssetPtr=&/Slinky_Susp_Adjust_Lower_Large_02}}
 {id=WavePlayer_17, properties {soundWaveAssetPtr=&/Slinky_Susp_Adjust_Lower_Large_03}}
 {id=WavePlayer_18, properties {soundWaveAssetPtr=&/Slinky_Susp_Adjust_Lower_Large_04}}
 {id=WavePlayer_19, properties {soundWaveAssetPtr=&Slinky_Susp_Adjust_Body_Rattle_Large_01}}
 {id=WavePlayer_2, properties {soundWaveAssetPtr=&/Slinky_Susp_Adjust_Body_Rattle_Large_03}}
 {id=WavePlayer_20, properties {soundWaveAssetPtr=&Slinky_Susp_Adjust_Body_Rattle_Large_02}}
 {id=WavePlayer_21, properties {soundWaveAssetPtr=&Slinky_Susp_Adjust_Body_Rattle_Large_03}}
 {id=WavePlayer_22, properties {soundWaveAssetPtr=&/Slinky_Susp_Adjust_Body_Rattle_Large_04}}
 {id=WavePlayer_23, properties {soundWaveAssetPtr=&/Slinky_Susp_Adjust_Body_Rattle_Large_05}}
 {id=WavePlayer_24, properties {soundWaveAssetPtr=&/Slinky_Susp_Adjust_Body_Rattle_Large_06}}
 {id=WavePlayer_25, properties {soundWaveAssetPtr=&Slinky_Susp_Adjust_Body_Rattle_Large_07}}
 {id=WavePlayer_26, properties {soundWaveAssetPtr=&Slinky_Susp_Adjust_Body_Rattle_Large_08}}
 {id=WavePlayer_27, properties {soundWaveAssetPtr=&Slinky_Susp_Adjust_Body_Rattle_Large_09}}
 {id=WavePlayer_28, properties {soundWaveAssetPtr=&Slinky_Susp_Adjust_Body_Rattle_Large_10}}
 {id=WavePlayer_29, properties {soundWaveAssetPtr=&Slinky_Susp_Adjust_Body_Rattle_Large_11}}
 {id=WavePlayer_3, properties {soundWaveAssetPtr=&Slinky_Susp_Adjust_Lower_Large_01}}
 {id=WavePlayer_4, properties {soundWaveAssetPtr=&Slinky_Susp_Adjust_Lower_Large_02}}
 {id=WavePlayer_5, properties {soundWaveAssetPtr=&Slinky_Susp_Adjust_Lower_Large_03}}
 {id=WavePlayer_6, properties {soundWaveAssetPtr=&Slinky_Susp_Adjust_Lower_Large_04}}
 {id=WavePlayer_7, properties {soundWaveAssetPtr=&Slinky_Susp_Adjust_Body_Rattle_Large_04}}
 {id=WavePlayer_8, properties {soundWaveAssetPtr=&Slinky_Susp_Adjust_Body_Rattle_Large_05}}
 {id=WavePlayer_9, properties {soundWaveAssetPtr=&Slinky_Susp_Adjust_Body_Rattle_Large_06}}
]
properties {
 bOverrideConcurrency=true
 maxDistance=12000
 priority=0.8
 soundClassObject=&Fort_Vehicles_Exterior
 virtualizationMode=Disabled
 volumeMultiplier=0.2
 concurrencyOverrides {bLimitToOwner=true, maxCount=2, resolutionRule=StopOldest}
}
