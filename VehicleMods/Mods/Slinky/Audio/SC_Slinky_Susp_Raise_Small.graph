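# Sound cue for the small suspension-raise sound of the Slinky vehicle mod, in the exterior vehicle sound class.
# An IsLocal branch selects first-person attenuation when true, and third-person attenuation via a mixer when false or unset.
# Each path randomly picks from six wave variations with pitch modulation; the first-person path stacks two modulators.
# A concurrency override caps it at two instances per owner, stopping the oldest.
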SoundCue /VehicleMods/Mods/Slinky/Audio/SC_Slinky_Susp_Raise_Small
root=Branch_0
nodes [
 {id=Attenuation_0, children [Modulator_0], properties {attenuationSettings=&/ATT_Slinky_1P}}
 {id=Attenuation_1, children [Modulator_1], properties {attenuationSettings=&/ATT_Slinky_3P}}
 {id=Branch_0, children [Attenuation_0, Mixer_1, Mixer_1], properties {boolParameterName=IsLocal}}
 {id=Mixer_1, children [Attenuation_1], properties {inputVolume [1]}}
 {id=Modulator_0, children [Modulator_2]}
 {id=Modulator_1, children [Random_0], properties {pitchMax=1.1, volumeMax=1, volumeMin=1}}
 {id=Modulator_2, children [Random_1], properties {pitchMax=1.1, volumeMax=1, volumeMin=1}}
 {
  id=Random_0
  children [WavePlayer_3,WavePlayer_4,WavePlayer_5,WavePlayer_6,WavePlayer_7,WavePlayer_8]
  properties {bShouldExcludeFromBranchCulling=true, weights [1,1,1,1,1,1]}
 }
 {
  id=Random_1
  children [WavePlayer_0, WavePlayer_1, WavePlayer_2, WavePlayer_9, WavePlayer_10, WavePlayer_11]
  properties {bShouldExcludeFromBranchCulling=true, weights [1,1,1,1,1,1]}
 }
 {id=WavePlayer_0, properties {soundWaveAssetPtr=&/Slinky_Susp_Adjust_Raise_Small_01}}
 {id=WavePlayer_1, properties {soundWaveAssetPtr=&/Slinky_Susp_Adjust_Raise_Small_02}}
 {id=WavePlayer_10, properties {soundWaveAssetPtr=&/Slinky_Susp_Adjust_Raise_Small_05}}
 {id=WavePlayer_11, properties {soundWaveAssetPtr=&/Slinky_Susp_Adjust_Raise_Small_06}}
 {id=WavePlayer_2, properties {soundWaveAssetPtr=&/Slinky_Susp_Adjust_Raise_Small_03}}
 {id=WavePlayer_3, properties {soundWaveAssetPtr=&Slinky_Susp_Adjust_Raise_Small_01}}
 {id=WavePlayer_4, properties {soundWaveAssetPtr=&Slinky_Susp_Adjust_Raise_Small_02}}
 {id=WavePlayer_5, properties {soundWaveAssetPtr=&Slinky_Susp_Adjust_Raise_Small_03}}
 {id=WavePlayer_6, properties {soundWaveAssetPtr=&/Slinky_Susp_Adjust_Raise_Small_04}}
 {id=WavePlayer_7, properties {soundWaveAssetPtr=&Slinky_Susp_Adjust_Raise_Small_05}}
 {id=WavePlayer_8, properties {soundWaveAssetPtr=&Slinky_Susp_Adjust_Raise_Small_06}}
 {id=WavePlayer_9, properties {soundWaveAssetPtr=&Slinky_Susp_Adjust_Raise_Small_04}}
]
properties {
 bOverrideConcurrency=true
 maxDistance=12000
 priority=0.8
 soundClassObject=&Fort_Vehicles_Exterior
 virtualizationMode=Disabled
 volumeMultiplier=0.5
 concurrencyOverrides {bLimitToOwner=true, maxCount=2, resolutionRule=StopOldest}
}
