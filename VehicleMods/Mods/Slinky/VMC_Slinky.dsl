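; Slinky vehicle mod that gives a Fortnite Dagwood vehicle lowrider hydraulics: a charged jump and steering-driven suspension tilt.
; Holding handbrake charges the jump, lowering springs and shifting camera FOV; release launches along vehicle up if grounded, dry and off cooldown.
; While occupied, a looping timer lifts or lowers springs by steering and speed, pausing above a speed limit, and detects landings.
; Tracked per-spring offsets make changes reversible; cosmetic audio and surface-aware VFX accompany actions.
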(event EventBeginPlay
  ; Begin Play
  (|CallParentFunction)
  (CallFunction|CacheVehicleVariables)
  (bind _owner_vehicle (Variables|Default|GetOwner%20Vehicle)
    (:then
      (EventDispatchers|BindEventtoOnStartHandbrakeDelegate _owner_vehicle (EventDispatchers|CreateEvent :!event Custom|Start%20Handbrake))
      (Vehicle|BindEventtoOnPawnEnteredVehicleSeat _owner_vehicle (EventDispatchers|CreateEvent :!event Custom|On%20Pawn%20Enter%20Vehicle))
      (Vehicle|BindEventtoOnPawnExitedVehicleSeat _owner_vehicle (EventDispatchers|CreateEvent :!event Custom|On%20Pawn%20Exit%20Vehicle))
      (Utilities|FlowControl|Sequence
        (:then_0
          (bind _returnvalue (GameplayEventRouter|GetEventRouter (Utilities|GetClass (Components|GetOwner)) (Components|GetOwner)))
          (GameplayEventRouter|StartListeningToEvent _returnvalue "/Script/FortniteGame.VehicleCosmeticsFinishedPayload"
            (:OnEventReceived
              (CallFunction|UpdateVehicleAnimInstance))))
        (:then_1
          (GameplayCue|ExecuteGameplayCue%28Burst%29Local 0 (Variables|Default|GetGC_ApplyMod) (Ability|GameplayCue|MakeGameplayCueParameters 0.0))
          (CallFunction|ApplySlinkyUI)
          (CallFunction|AppendSlinkyUIActionLabel)
          (Variables|Jumping|SetJump%20Charge (Variables|Jumping|GetMin%20Jump%20Charge%20Percentage)))))
    (:else)))

(event EventEndPlay
  (params (EndPlayReason :type EEndPlayReason%20Enum :default "Destroyed"))
  (bind _owner_vehicle (Variables|Default|GetOwner%20Vehicle))
  ;[0 0.03012 0.8438 0.5] End Play
  (|CallParentFunction EndPlayReason)
  (Utilities|Time|ClearandInvalidateTimerbyHandle (Variables|Default|GetTilt%20Check%20Timer))
  (EventDispatchers|UnbindEventfromOnStartHandbrakeDelegate _owner_vehicle (EventDispatchers|CreateEvent :!event Custom|Start%20Handbrake))
  (EventDispatchers|UnbindEventfromOnStopHandbrakeDelegate _owner_vehicle (EventDispatchers|CreateEvent :!event Custom|Stop%20Handbrake))
  (Vehicle|UnbindEventfromOnPawnEnteredVehicleSeat _owner_vehicle (EventDispatchers|CreateEvent :!event Custom|On%20Pawn%20Enter%20Vehicle))
  (Vehicle|UnbindEventfromOnPawnExitedVehicleSeat _owner_vehicle (EventDispatchers|CreateEvent :!event Custom|On%20Pawn%20Exit%20Vehicle))
  (Vehicle|UnbindEventfromOnSpringCompressionDelegate _owner_vehicle (EventDispatchers|CreateEvent :!event Custom|Reset%20Jump%20Spring%20Offsets))
  (Utilities|Time|ClearandInvalidateTimerbyHandle (Variables|Default|GetJump%20Charge%20Timer))
  (Utilities|Time|ClearandInvalidateTimerbyHandle (Variables|Default|GetSecond%20Lower%20Audio%20Timer)))

(event Custom|Cache%20Vehicle%20Variables
  ;[0.6354 0.5963 0] Setup the initial vehicle references
  (bind _asfort_dagwood_vehicle (Utilities|Casting|CastToFortDagwoodVehicle (Components|GetOwner)))
  (bind _output_get (Variables|Default|SetOwner%20Vehicle _asfort_dagwood_vehicle))
  (bind _asfort_vehicle_anim_instance (Utilities|Casting|CastToFortVehicleAnimInstance (Components|SkeletalMesh|GetAnimInstance (Variables|/Script/FortniteGame.FortAthenaVehicle|GetSkeletalMesh _output_get))))
  (Variables|Default|SetCurrent%20Vehicle%20Anim%20Instance _asfort_vehicle_anim_instance)
  (Variables|Default|SetMax%20Vehicle%20Steering%20Angle (* (Variables|/Script/FortniteGame.FortPhysicsVehicleConfigs|GetMinSpeedSteeringAngle (Variables|/Script/ValetRuntime.FortDagwoodVehicle|GetFortDagwoodVehicleConfigs _output_get)) (Variables|/Script/FortniteGame.FortPhysicsVehicleConfigs|GetSteerBrakeMultiplier (Variables|/Script/ValetRuntime.FortDagwoodVehicle|GetFortDagwoodVehicleConfigs _output_get)))))

(event Custom|Start%20Handbrake
  (bind _owner_vehicle (Variables|Default|GetOwner%20Vehicle))
  ;[1 0 0.6785] Handle Hyrdaulics Jump
  ;;[0 0.1266 0.15] pressing and releasing the handbrake initiates the hydraulic jump
  (if (not (Variables|Default|GetHandBrake%20Active))
    (EventDispatchers|UnbindEventfromOnStartHandbrakeDelegate _owner_vehicle (delegate Custom|Start%20Handbrake))
    (EventDispatchers|BindEventtoOnStopHandbrakeDelegate _owner_vehicle (delegate Custom|Stop%20Handbrake))
    (CallFunction|TriggerHydraulicCharge)
    (Variables|Default|SetHandBrake%20Active true)
    (Networking|CanExecuteCosmeticEvents
      (:True
        ;;;[0.07534 0.4635 0] Hydraulic Audio
        (bind _returnvalue (Audio|SpawnSoundAttached (Variables|Audio|GetJump%20Prime%20Sound) (Variables|/Script/Engine.Actor|GetRootComponent _owner_vehicle)))
        (Utilities|SetAudioParameters _returnvalue)
        (bind _returnvalue_2 (Utilities|Time|SetTimerbyEvent (delegate Custom|Play%20Hydraulic%20Jump%20Charged%20Sound) (Variables|Jumping|GetSecond%20Hydraulic%20Audio%20Delay)))
        (Variables|Default|SetSecond%20Lower%20Audio%20Timer _returnvalue_2)))))

(event Custom|Play%20Hydraulic%20Jump%20Charged%20Sound
  ;[1 0 0.6785] Handle Hyrdaulics Jump
  ;;[0 0.1266 0.15] pressing and releasing the handbrake initiates the hydraulic jump
  (bind _returnvalue (Audio|SpawnSoundAttached (Variables|Audio|GetJump%20Charged%20Sound) (Variables|/Script/Engine.Actor|GetRootComponent (Variables|Default|GetOwner%20Vehicle))))
  (Utilities|SetAudioParameters _returnvalue))

(event Custom|Stop%20Handbrake
  (bind _owner_vehicle (Variables|Default|GetOwner%20Vehicle))
  ;[1 0 0.6785] Handle Hyrdaulics Jump
  ;;[0 0.1266 0.15] pressing and releasing the handbrake initiates the hydraulic jump
  (if (Variables|Default|GetHandBrake%20Active)
    (EventDispatchers|UnbindEventfromOnStopHandbrakeDelegate _owner_vehicle (delegate Custom|Stop%20Handbrake))
    (Variables|Default|SetHandBrake%20Active false)
    (CallFunction|ResetSpringOffsets (Variables|Jumping|GetChargeJumpOffsets))
    (if (Math|Boolean|ANDBoolean (Vehicle|WheelsonGround _owner_vehicle) (not (Class|FortAthenaVehicle|IsOverlappingWaterBody _owner_vehicle)) (not (Utilities|Time|IsValidTimerHandle (Variables|Default|GetJump%20Cooldown%20Timer))))
      (bind _returnvalue (Utilities|Time|SetTimerbyEvent (EventDispatchers|CreateEvent :!event Custom|Reset%20Jump%20Cooldown) (Variables|Default|GetJump%20Cooldown%20Time)))
      (Variables|Default|SetJump%20Cooldown%20Timer _returnvalue)
      (CallFunction|LowriderJump)
      (else
        (CallFunction|ClearChargeJump)
        (EventDispatchers|BindEventtoOnStartHandbrakeDelegate (Variables|Default|GetOwner%20Vehicle) (EventDispatchers|CreateEvent :!event Custom|Start%20Handbrake))))))

(event Custom|Reset%20Jump%20Cooldown
  ;[1 0 0.6785] Handle Hyrdaulics Jump
  (Utilities|Time|ClearandInvalidateTimerbyHandle (Variables|Default|GetJump%20Cooldown%20Timer))
  (EventDispatchers|BindEventtoOnStartHandbrakeDelegate (Variables|Default|GetOwner%20Vehicle) (EventDispatchers|CreateEvent :!event Custom|Start%20Handbrake)))

(event Custom|Lowrider%20Jump
  (bind _owner_vehicle (Variables|Default|GetOwner%20Vehicle))
  ;[1 0 0.6785] Handle Hyrdaulics Jump
  (Utilities|IsValid _owner_vehicle
    (:Is%20Valid
      (Variables|Default|SetIs%20Jumping true)
      (FortPhysicsSimulationLibrary|SetPhysicsLinearVelocity _owner_vehicle (* (* (Math|Vector|Normalize (Transformation|GetActorUpVector _owner_vehicle)) (Variables|Jumping|GetJump%20Velocity)) (Variables|Jumping|GetJump%20Charge)) true "None" 0 "(MaxSimSizeAllowed=Large,bRejectEmbeddedInLandscape=True)")
      (CallFunction|ClearChargeJump)
      (CallFunction|PlayJumpFX)
      (CallFunction|FakeSpringJump))))

(event Custom|Vehicle%20Jump%20Landed
  ;[1 0 0.6785] Handle Hyrdaulics Jump
  (Networking|CanExecuteCosmeticEvents
    (:True
      (CallFunction|PlayJumpLandSound))))

(event Custom|Play%20Jump%20FX
  (bind _returnvalue (Class|FortVehicleInterface|GetSkeletalMeshComponent (Variables|Default|GetOwner%20Vehicle)))
  (bind _returnvalue_location (Math|Transform|BreakTransform (Transformation|GetBoneTransform (Class|FortVehicleInterface|GetSkeletalMeshComponent (Variables|Default|GetOwner%20Vehicle)) "Root" "RTS_World")))
  (bind _owner_vehicle (Variables|Default|GetOwner%20Vehicle))
  (bind _jump_vfx (Variables|Jumping|GetJump%20VFX))
  ;[1 0 0.6785] Handle Hyrdaulics Jump
  (Networking|CanExecuteCosmeticEvents
    (:True
      (Utilities|FlowControl|Sequence
        (:then_0
          ;;[0.5838 0.599 0] Jump SFX
          (bind _returnvalue_2 (Audio|SpawnSoundAttached :Sound (Variables|Audio|GetJump%20Sound) :AttachToComponent (Variables|/Script/Engine.Actor|GetRootComponent (Variables|Default|GetOwner%20Vehicle)) :AttachPointName "None" :LocationType "KeepRelativeOffset" :bStopWhenAttachedToDestroyed true :VolumeMultiplier 1.0 :PitchMultiplier 1.0 :StartTime 0.0 :bAutoDestroy true))
          (Utilities|SetAudioParameters _returnvalue_2))
        (:then_1
          ;;[0 0.3542 0.03916] Jump VFX
          (Utilities|IsValid _jump_vfx
            (:Is%20Valid
              (if (Variables|Jumping|GetUseOneJumpFX)
                (bind _returnvalue_3 (Effects|Components|ParticleSystem|SpawnFXAtLocation (Variables|Jumping|GetJump%20VFX) _returnvalue_location))
                (Utilities|Array|Clear (Variables|Jumping|GetJumpWheelAvg))
                (for _array_element (Vehicle|FX|GetWheelFXSockets (Variables|Default|GetOwner%20Vehicle))
                  (Utilities|Array|Add (Variables|Jumping|GetJumpWheelAvg) (.Offset (Utilities|Struct|BreakVehicleWheelFXSocketInfo _array_element))))
                (bind _surface_type (CallFunction|DetermineWorldMaterialUnderWheel (Math|Vector|GetVectorArrayAverage (Variables|Jumping|GetJumpWheelAvg))))
                (Effects|Components|ParticleSystem|SetBoolParameter _returnvalue_3 "JumpOnSnowOrIce" (or (Utilities|Enum|Equal(Enum) _surface_type "SurfaceType23") (Utilities|Enum|Equal(Enum) _surface_type "SurfaceType24")))
                (else
                  (for _array_element_2 (Vehicle|FX|GetWheelFXSockets _owner_vehicle)
                    (bind _returnvalue_4 (Effects|Components|ParticleSystem|SpawnFXAtLocation _jump_vfx (+ (Transformation|GetSocketLocation _returnvalue (Utilities|Struct|BreakVehicleWheelFXSocketInfo _array_element_2)) (.Offset (Utilities|Struct|BreakVehicleWheelFXSocketInfo _array_element_2))) (Transformation|GetActorRotation _owner_vehicle)))
                    (bind _surface_type_2 (CallFunction|DetermineWorldMaterialUnderWheel (+ (Transformation|GetSocketLocation _returnvalue (Utilities|Struct|BreakVehicleWheelFXSocketInfo _array_element_2)) (.Offset (Utilities|Struct|BreakVehicleWheelFXSocketInfo _array_element_2)))))
                    (Effects|Components|ParticleSystem|SetBoolParameter _returnvalue_4 "JumpOnSnowOrIce" (or (Utilities|Enum|Equal(Enum) _surface_type_2 "SurfaceType23") (Utilities|Enum|Equal(Enum) _surface_type_2 "SurfaceType24")))))))))))))

(event Custom|Play%20Jump%20Land%20Sound
  ;[1 0 0.6785] Handle Hyrdaulics Jump
  (bind _returnvalue (Audio|SpawnSoundAttached :Sound (Variables|Audio|GetJump%20Land%20Sound) :AttachToComponent (Variables|/Script/Engine.Actor|GetRootComponent (Variables|Default|GetOwner%20Vehicle)) :AttachPointName "None" :LocationType "KeepRelativeOffset" :bStopWhenAttachedToDestroyed true :VolumeMultiplier 1.0 :PitchMultiplier 1.0 :StartTime 0.0 :bAutoDestroy true))
  (Utilities|SetAudioParameters _returnvalue))

(event Custom|Start%20Charging%20Jump
  (bind _owner_vehicle (Variables|Default|GetOwner%20Vehicle))
  (bind _returnvalue (Class|FortVehicleInterface|GetSkeletalMeshComponent _owner_vehicle))
  (bind _jumpchargestartindicatorfx (Variables|Jumping|GetJumpChargeStartIndicatorFX))
  ;[1 0 0.6785] Handle Hyrdaulics Jump
  ;;[0 0.04184 0.15] Charged Jump Actions
  (if (not (Utilities|Time|IsValidTimerHandle (Variables|Default|GetJump%20Cooldown%20Timer)))
    (Utilities|FlowControl|Sequence
      (:then_0
        (Networking|CanExecuteCosmeticEvents
          (:True
            (Utilities|IsValid _jumpchargestartindicatorfx
              (:Is%20Valid
                (bind _returnvalue_2 (Niagara|SpawnSystemAttached _jumpchargestartindicatorfx _returnvalue "None" "0, 0, 0" "0, 0, 0" "KeepRelativeOffset" true))
                (Niagara|SetNiagaraSkeletalMeshComponent _returnvalue_2 "Skeletal Mesh" _returnvalue))))))
      (:then_1
        (CallFunction|IncreaseJumpCharge)
        (bind _returnvalue_3 (Utilities|Time|SetTimerbyEvent (delegate Custom|Increase%20Jump%20Charge) (Variables|Jumping|GetJump%20Charge%20increase%20rate) true true))
        (Variables|Default|SetJump%20Charge%20Timer _returnvalue_3)
        (if (Vehicle|WheelsonGround (Variables|Default|GetOwner%20Vehicle))
          (CallFunction|OffsetSprings (Variables|Jumping|GetChargeJumpOffsets) (Variables|Default|GetCharge%20Jump%20Lower%20Amount)))))))

(event Custom|Increase%20Jump%20Charge
  (bind _min_jump_charge_percentage (Variables|Jumping|GetMin%20Jump%20Charge%20Percentage))
  (bind _jump_charge_increase_increment (Variables|Jumping|GetJump%20Charge%20Increase%20Increment))
  (bind _jump_charge (Variables|Jumping|GetJump%20Charge))
  (bind _returnvalue (+ _jump_charge _jump_charge_increase_increment))
  (bind _returnvalue_2 (Math|Float|Clamp(Float) _returnvalue _min_jump_charge_percentage))
  (bind _owner_vehicle (Variables|Default|GetOwner%20Vehicle))
  ;[1 0 0.6785] Handle Hyrdaulics Jump
  ;;[0 0.04184 0.15] Charged Jump Actions
  (if (not (Vehicle|VehicleIsinAir _owner_vehicle))
    (bind _output_get (Variables|Jumping|SetJump%20Charge _returnvalue_2))
    (Variables|/Script/FortniteGame.FortAthenaVehicle|SetCameraFOVOffset (* _output_get (Variables|Jumping|GetJumpFOVOffset)) _owner_vehicle)
    (Default|CallIncreaseCharge _returnvalue_2)))

(event Custom|Clear%20Charge%20Jump
  (bind _second_lower_audio_timer (Variables|Default|GetSecond%20Lower%20Audio%20Timer))
  ;[1 0 0.6785] Handle Hyrdaulics Jump
  ;;[0 0.04184 0.15] Charged Jump Actions
  (Variables|Jumping|SetJump%20Charge (Variables|Jumping|GetMin%20Jump%20Charge%20Percentage))
  (Variables|/Script/FortniteGame.FortAthenaVehicle|SetCameraFOVOffset 0.0 (Variables|Default|GetOwner%20Vehicle))
  (Utilities|Time|ClearandInvalidateTimerbyHandle (Variables|Default|GetJump%20Charge%20Timer))
  (Default|CallClearCharge)
  (if (Utilities|Time|IsValidTimerHandle _second_lower_audio_timer)
    (Utilities|Time|ClearandInvalidateTimerbyHandle _second_lower_audio_timer)))

(event Custom|Fake%20Spring%20Jump
  (bind _owner_vehicle (Variables|Default|GetOwner%20Vehicle))
  ;[1 0 0.6785] Handle Hyrdaulics Jump
  ;;[0 0.04184 0.15] Charged Jump Actions
  (if (and (Utilities|IsValid (Variables|Default|GetCurrent%20Vehicle%20Anim%20Instance) :!class /Script/Engine.KismetSystemLibrary) (Utilities|IsValid _owner_vehicle :!class /Script/Engine.KismetSystemLibrary))
    (CallFunction|PerformJumpSpringOffsets (Variables|Jumping|GetJump%20Spring%20Offset%20Amount))
    (Utilities|FlowControl|Delay (Variables|Jumping|GetJump%20Spring%20Increase%20Delay)) ; add a delay for the vehicle to get off the ground before triggering the spring compression reset
    (Vehicle|BindEventtoOnSpringCompressionDelegate _owner_vehicle (delegate Custom|Reset%20Jump%20Spring%20Offsets))))

(event Custom|Reset%20Jump%20Spring%20Offsets
  (bind _owner_vehicle (Variables|Default|GetOwner%20Vehicle))
  ;[1 0 0.6785] Handle Hyrdaulics Jump
  (if (and (Utilities|IsValid _owner_vehicle :!class /Script/Engine.KismetSystemLibrary) (Utilities|IsValid (Variables|Default|GetCurrent%20Vehicle%20Anim%20Instance) :!class /Script/Engine.KismetSystemLibrary))
    (Vehicle|UnbindEventfromOnSpringCompressionDelegate _owner_vehicle (delegate Custom|Reset%20Jump%20Spring%20Offsets))
    (CallFunction|ResetSpringOffsets (Variables|Jumping|GetJump%20Spring%20Offsets))))

(event Custom|Trigger%20Hydraulic%20Charge
  (CallFunction|StartChargingJump))

(event Custom|On%20Pawn%20Enter%20Vehicle
  (params (Vehicle :type Fort%20Vehicle%20Interface%20Interface%20%28by%20ref%29) (PlayerPawn :type Fort%20Pawn%20Object%20Reference) (SeatIndex :type Integer))
  ;[0 1 0.08787] Enter / Exit Vehicle
  (Variables|Default|SetPlayerInVehicle true)
  (if (not (Utilities|Time|IsValidTimerHandle (Variables|Default|GetTilt%20Check%20Timer)))
    (bind _returnvalue (Utilities|Time|SetTimerbyEvent (EventDispatchers|CreateEvent :!event Custom|Update%20Suspension%20Springs%20Tilt) (Variables|Tilting|GetTilt%20Update%20Rate) true true))
    (Variables|Default|SetTilt%20Check%20Timer _returnvalue)
    (if (== SeatIndex 0)
      (Utilities|FlowControl|Delay (Variables|Tilting|GetTilt%20Audio%20Initial%20Delay))
      (Variables|Tilting|SetBlock%20Tilt%20Audio false))))

(event Custom|On%20Pawn%20Exit%20Vehicle
  (params (Vehicle :type Fort%20Vehicle%20Interface%20Interface%20%28by%20ref%29) (PlayerPawn :type Fort%20Pawn%20Object%20Reference) (SeatIndex :type Integer))
  ;[0 1 0.08787] Enter / Exit Vehicle
  (Variables|Default|SetPlayerInVehicle (Vehicle|ContainsPlayers Vehicle))
  (if (== SeatIndex 0)
    (CallFunction|ManuallyUntiltSprings)
    (Utilities|Time|ClearandInvalidateTimerbyHandle (Variables|Default|GetTilt%20Check%20Timer))
    (Variables|Tilting|SetBlock%20Tilt%20Audio true)))

(event Custom|Play%20Tilt%20Sound
  ;[0.474 0.1582 0] Play raise and lower sounds
  (Networking|CanExecuteCosmeticEvents
    (:True
      (if (not (Variables|Tilting|GetBlock%20Tilt%20Audio))
        (bind _returnvalue (Audio|SpawnSoundAttached (select (Variables|Tilting|GetWas%20Lowered) (Variables|Audio|GetLower%20Sound) (select (<= (Vehicle|GetForwardSpeedKmh (Variables|Default|GetOwner%20Vehicle)) (Variables|Default|GetMedium%20Speed%20Activation)) (Variables|Audio|GetRaise%20Large%20Sound) (Variables|Audio|GetRaise%20Small%20Sound))) (Variables|/Script/Engine.Actor|GetRootComponent (Variables|Default|GetOwner%20Vehicle))))
        (Utilities|SetAudioParameters _returnvalue)))))

(event Custom|Update%20Suspension%20Springs%20Tilt
  (bind _spring_index (CallFunction|IsSteeringLeft))
  ;[0 0.08852 1] Handle Tilting
  (CallFunction|CheckifJumpLanded)
  (if (not (CallFunction|IsOverMaxModSpeed))
    (if (and (Variables|Default|GetPlayerInVehicle) (CallFunction|CanModPerformActions))
      (if (not (Math|Float|InRange(Float) (Vehicle|GetSteeringAngle (Variables|Default|GetOwner%20Vehicle)) (Variables|Tilting|GetMinSteerAngleTrigger) (Variables|Tilting|GetMaxSteerAngleTrigger)))
        (if (or (Variables|Default|GetWasSteeringStraight) (!= (Variables|Default|GetSteering%20Left) _spring_index))
          (CallFunction|PlayTiltSound)
          (else
            _))
        (Variables|Default|SetSteering%20Left _spring_index)
        (Variables|Default|SetWasSteeringStraight false)
        (Variables|Tilting|SetWas%20Raised false)
        (Variables|Tilting|SetWas%20Lowered false)
        (CallFunction|BlendTiltSprings (Math|Float|SelectFloat (Variables|Tilting|GetMax%20Hydraulic%20Spring%20Height) (Variables|Tilting|GetMid%20Hydraulic%20Spring%20Height) (<= (Vehicle|GetForwardSpeedKmh (Variables|Default|GetOwner%20Vehicle)) (Variables|Default|GetMedium%20Speed%20Activation))))
        (elif (not (Variables|Default|GetWasSteeringStraight))
          (CallFunction|PlayTiltSound)
          (Variables|Default|SetWasSteeringStraight true)
          (Variables|Default|SetSteering%20Left false)
          (CallFunction|UntiltAllSprings)
          (Variables|Tilting|SetWas%20Lowered false)
          (else
            _)))
      (else
        _))
    (elif (not (Utilities|Time|IsValidTimerHandle (Variables|Default|GetMax%20Speed%20Tilt%20Delay%20Timer)))
      (bind _returnvalue (Utilities|Time|SetTimerbyEvent (EventDispatchers|CreateEvent :!event Custom|UnBlock%20Tilting) (Variables|Default|GetMax%20Speed%20Tilting%20Delay)))
      (Variables|Default|SetMax%20Speed%20Tilt%20Delay%20Timer _returnvalue)
      (Variables|Default|SetBlock%20Tilting true)
      (Variables|Tilting|SetBlock%20Tilt%20Audio true)
      (else
        (Utilities|Time|ClearandInvalidateTimerbyHandle (Variables|Default|GetMax%20Speed%20Tilt%20Delay%20Timer))))))

(event Custom|UnBlock%20Tilting
  ;[0 0.08852 1] Handle Tilting
  (Variables|Tilting|SetBlock%20Tilt%20Audio false)
  (Variables|Default|SetBlock%20Tilting false)
  (Utilities|Time|ClearandInvalidateTimerbyHandle (Variables|Default|GetMax%20Speed%20Tilt%20Delay%20Timer)))

(event Custom|Manually%20Untilt%20Springs
  ;[0 0.08852 1] Handle Tilting
  (CallFunction|ResetSpringOffsets (Variables|Tilting|GetVehicle%20Spring%20Offsets)))

(event Custom|Update%20Vehicle%20Anim%20Instance
  (bind _asfort_vehicle_anim_instance (Utilities|Casting|CastToFortVehicleAnimInstance (Components|SkeletalMesh|GetAnimInstance (Variables|/Script/FortniteGame.FortAthenaVehicle|GetSkeletalMesh (Variables|Default|GetOwner%20Vehicle))))
    (:then
      (Variables|Default|SetCurrent%20Vehicle%20Anim%20Instance _asfort_vehicle_anim_instance))
    (:CastFailed
      (Development|PrintString "Vehicle Anim Instance Failed To Update"))))

(fn BlendTiltSprings
  (params (Max%20Spring%20Height :type Float%20%28double-precision%29))
  (bind _vehicle_spring_offsets (Variables|Tilting|GetVehicle%20Spring%20Offsets))
  (if (and (Utilities|IsValid (Variables|Default|GetOwner%20Vehicle) :!class /Script/Engine.KismetSystemLibrary) (Utilities|IsValid (Variables|Default|GetCurrent%20Vehicle%20Anim%20Instance) :!class /Script/Engine.KismetSystemLibrary))
    (for (_ _array_index) _vehicle_spring_offsets
      (if (not (Variables|Default|GetBlock%20Tilting))
        (if (== (Math|Integer|%25%28Integer%29 (Utilities|Struct|BreakSlinkyOffsets (Utilities|Array|Get(aref) _vehicle_spring_offsets _array_index)) 2) (Math|Integer|SelectInt 0 1 (CallFunction|IsSteeringLeft)))
          ;[0 0.01619 0.2396 0.5] increase the suspension springs that are not currently at the maximum length
          (if (not (> (.Offset_5_57C2233F4058AA4CB940A6815D847DC8 (Utilities|Struct|BreakSlinkyOffsets (Utilities|Array|Get(aref) _vehicle_spring_offsets _array_index))) Max%20Spring%20Height))
            (Variables|Tilting|SetWas%20Raised true)
            (CallFunction|LiftSpring (Utilities|Array|Get(aref) _vehicle_spring_offsets _array_index) (Math|Float|SelectFloat (Variables|Tilting|GetHydraulic%20Spring%20Increase) (- Max%20Spring%20Height (.Offset_5_57C2233F4058AA4CB940A6815D847DC8 (Utilities|Struct|BreakSlinkyOffsets (Utilities|Array|Get(aref) _vehicle_spring_offsets _array_index)))) (< (+ (Variables|Tilting|GetHydraulic%20Spring%20Increase) (.Offset_5_57C2233F4058AA4CB940A6815D847DC8 (Utilities|Struct|BreakSlinkyOffsets (Utilities|Array|Get(aref) _vehicle_spring_offsets _array_index)))) Max%20Spring%20Height)))
            (else
              ;[0.15 0.141 0] If the spring is currently higher than the new maximum length, clamp/lower it to the new maximum
              (Variables|Tilting|SetWas%20Raised true)
              (CallFunction|LiftSpring (Utilities|Array|Get(aref) _vehicle_spring_offsets _array_index) (Math|Float|SelectFloat (- Max%20Spring%20Height) (* (* (Variables|Tilting|GetHydraulic%20Spring%20Increase) (Variables|Default|GetDecrease%20Spring%20Multiplier)) -1.0) (< (- (.Offset_5_57C2233F4058AA4CB940A6815D847DC8 (Utilities|Struct|BreakSlinkyOffsets (Utilities|Array|Get(aref) _vehicle_spring_offsets _array_index))) (* (Variables|Tilting|GetHydraulic%20Spring%20Increase) (Variables|Default|GetDecrease%20Spring%20Multiplier))) Max%20Spring%20Height)))))
          (else
            _))
        (else
          _))
      ;[0 0.15 0.04179] lower the suspension springs that should not be lifted back to their default length
      (CallFunction|LowerSpring (Utilities|Array|Get(aref) _vehicle_spring_offsets _array_index))
      (Variables|Tilting|SetWas%20Lowered true))))

(fn CalculateUntiltAmount
  (params (Spring :type Slinky%20Offsets%20Structure))
  (return (Math|Float|SelectFloat (* (Variables|Tilting|GetHydraulic%20Spring%20Increase) -1.0) (* (Utilities|Struct|BreakSlinkyOffsets Spring) -1.0) (>= (- (Utilities|Struct|BreakSlinkyOffsets Spring) (Variables|Tilting|GetHydraulic%20Spring%20Increase)) 0))))

(fn CanModPerformActions
  (bind _owner_vehicle (Variables|Default|GetOwner%20Vehicle))
  (Utilities|IsValid _owner_vehicle
    (:Is%20Valid
      (return (and (not (CallFunction|IsOverMaxModSpeed)) (Vehicle|WheelsonGround _owner_vehicle))))
    (:Is%20Not%20Valid
      (return false))))

(fn CheckIfJumpLanded
  (bind _output_get (LocalVariables|SetCurrent%20Vehicle%20Is%20In%20Air (Vehicle|VehicleIsinAir (Variables|Default|GetOwner%20Vehicle))))
  (if (!= _output_get (Variables|Default|GetIs%20In%20Air))
    (bind _output_get_2 (Variables|Default|SetIs%20In%20Air (LocalVariables|GetCurrent%20Vehicle%20Is%20In%20Air)))
    (if (and (not _output_get_2) (Variables|Default|GetIs%20Jumping))
      (Variables|Default|SetIs%20Jumping false)
      (CallFunction|VehicleJumpLanded))))

(fn DetermineWorldMaterialUnderWheel
  (params (Wheel%20Location :type Vector))
  (bind (_outhit _returnvalue) (Collision|LineTraceByChannel (Utilities|Operators|Add Wheel%20Location (Math|Vector|MakeVector 0.0 0.0 (Variables|Default|GetWorld%20Material%20Trace%20Height))) (Utilities|Operators|Add :B (Math|Vector|MakeVector 0.0 0.0 (* (Variables|Default|GetWorld%20Material%20Trace%20Height) -1.0))) "TraceTypeQuery10"))
  (if _returnvalue
    (return (Variables|/Script/PhysicsCore.PhysicalMaterial|GetSurfaceType (.PhysMat (Collision|BreakHitResult _outhit))))
    (else
      (return "SurfaceType_Default"))))

(fn GetExhaustSockets
  (params (Skeletal%20Mesh :type Skeletal%20Mesh%20Component%20Object%20Reference))
  (Utilities|FlowControl|Sequence
    (:then_0
      (for _array_element (Transformation|GetAllSocketNames Skeletal%20Mesh)
        (if (Utilities|String|Contains (Utilities|String|ToString(Name) _array_element) "exhaust")
          (Utilities|Array|Add (LocalVariables|GetLocal%20Exhaust%20Sockets) _array_element))))
    (:then_1
      (return (LocalVariables|GetLocal%20Exhaust%20Sockets)))))

(fn IsOverMaxModSpeed
  (bind _owner_vehicle (Variables|Default|GetOwner%20Vehicle))
  (Utilities|IsValid _owner_vehicle
    (:Is%20Valid
      (return (> (Vehicle|GetForwardSpeedKmh _owner_vehicle) (Variables|Default|GetMax%20Speed%20Deactivation))))
    (:Is%20Not%20Valid
      (return true))))

(fn IsSteeringLeft
  (return (>= (Vehicle|GetSteeringAngle (Variables|Default|GetOwner%20Vehicle)) 0)))

(fn LiftSpring
  (params (Spring :type Slinky%20Offsets%20Structure%20%28by%20ref%29) (Amount :type Float%20%28double-precision%29))
  (bind (_springindex_4_57b4a62d476e0143ada6a68828d759ff _offset_5_57c2233f4058aa4cb940a6815d847dc8) (Utilities|Struct|BreakSlinkyOffsets Spring))
  (CallFunction|ModifyVehicleSpringLength _springindex_4_57b4a62d476e0143ada6a68828d759ff Amount (Variables|Default|GetOwner%20Vehicle) (Variables|Default|GetCurrent%20Vehicle%20Anim%20Instance)) ; handle the spring that should increase
  (Utilities|Struct|SetmembersinSlinkyOffsets Spring (+ _offset_5_57c2233f4058aa4cb940a6815d847dc8 Amount))
  (return))

(fn LowerSpring
  (params (Spring :type Slinky%20Offsets%20Structure%20%28by%20ref%29))
  (bind (_springindex_4_57b4a62d476e0143ada6a68828d759ff _offset_5_57c2233f4058aa4cb940a6815d847dc8) (Utilities|Struct|BreakSlinkyOffsets Spring))
  (bind _current_untilt_amount (LocalVariables|GetCurrent%20Untilt%20Amount))
  (LocalVariables|SetCurrent%20Untilt%20Amount (Default|CalculateUntiltAmount Spring))
  (CallFunction|ModifyVehicleSpringLength _springindex_4_57b4a62d476e0143ada6a68828d759ff _current_untilt_amount (Variables|Default|GetOwner%20Vehicle) (Variables|Default|GetCurrent%20Vehicle%20Anim%20Instance)) ; handle all of the other springs (the springs that need to lower back down)
  (Utilities|Struct|SetmembersinSlinkyOffsets Spring (Math|Float|Clamp(Float) (+ _offset_5_57c2233f4058aa4cb940a6815d847dc8 _current_untilt_amount) 0.0 (Variables|Tilting|GetMax%20Hydraulic%20Spring%20Height)))
  (return))

(fn OffsetSprings
  (params (Offsets :type Array%20of%20Slinky%20Offsets%20Structures) (Amount :type Float%20%28double-precision%29))
  (for (_array_element _array_index) Offsets
    (CallFunction|ModifyVehicleSpringLength (Utilities|Struct|BreakSlinkyOffsets _array_element) Amount (Variables|Default|GetOwner%20Vehicle) (Variables|Default|GetCurrent%20Vehicle%20Anim%20Instance))
    (Utilities|Struct|SetmembersinSlinkyOffsets (Utilities|Array|Get(aref) Offsets _array_index) (+ (.Offset_5_57C2233F4058AA4CB940A6815D847DC8 (Utilities|Struct|BreakSlinkyOffsets (Utilities|Array|Get(aref) Offsets _array_index))) Amount))))

(fn PerformJumpSpringOffsets
  (params (Amount :type Float%20%28double-precision%29))
  (bind _jump_spring_offsets (Variables|Jumping|GetJump%20Spring%20Offsets))
  (for (_ _array_index) _jump_spring_offsets
    (bind _output_get (LocalVariables|SetActual%20Increment%20Amount (Math|Float|SelectFloat Amount 0.0 (<= (+ (Utilities|Struct|BreakSlinkyOffsets (Utilities|Array|Get(acopy) (Variables|Tilting|GetVehicle%20Spring%20Offsets) _array_index)) Amount) (Variables|Jumping|GetJump%20Spring%20Offset%20Amount)))))
    (CallFunction|ModifyVehicleSpringLength (Utilities|Struct|BreakSlinkyOffsets (Utilities|Array|Get(aref) _jump_spring_offsets _array_index)) _output_get (Variables|Default|GetOwner%20Vehicle) (Variables|Default|GetCurrent%20Vehicle%20Anim%20Instance))
    (Utilities|Struct|SetmembersinSlinkyOffsets (Utilities|Array|Get(aref) _jump_spring_offsets _array_index) (+ (Utilities|Struct|BreakSlinkyOffsets (Utilities|Array|Get(aref) _jump_spring_offsets _array_index)) (LocalVariables|GetActual%20Increment%20Amount)))))

(fn ResetSpringOffsets
  (params (Offsets :type Array%20of%20Slinky%20Offsets%20Structures))
  (if (Utilities|Array|IsNotEmpty Offsets)
    (for (_ _array_index) Offsets
      (CallFunction|ModifyVehicleSpringLength (Utilities|Struct|BreakSlinkyOffsets (Utilities|Array|Get(aref) Offsets _array_index)) (* (.Offset_5_57C2233F4058AA4CB940A6815D847DC8 (Utilities|Struct|BreakSlinkyOffsets (Utilities|Array|Get(aref) Offsets _array_index))) -1.0) (Variables|Default|GetOwner%20Vehicle) (Variables|Default|GetCurrent%20Vehicle%20Anim%20Instance))
      (Utilities|Struct|SetmembersinSlinkyOffsets (Utilities|Array|Get(aref) Offsets _array_index) 0.0))))

(fn UntiltAllSprings
  (bind _vehicle_spring_offsets (Variables|Tilting|GetVehicle%20Spring%20Offsets))
  (for (_ _array_index) _vehicle_spring_offsets
    (CallFunction|LowerSpring (Utilities|Array|Get(aref) _vehicle_spring_offsets _array_index))))
